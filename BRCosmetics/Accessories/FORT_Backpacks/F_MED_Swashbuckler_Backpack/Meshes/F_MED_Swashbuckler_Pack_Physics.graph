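# Physics asset for the Swashbuckler backpack cosmetic, providing collision and joint setup for the pack and its two dangling swords.
# It has three capsule bodies: one on backpack_root and one each on the left and right dyn sword bones.
# Each sword is constrained directly to backpack_root with both swing axes locked, so twist is the only angular axis left unlocked for sword motion.
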PhysicsAsset /BRCosmetics/Accessories/FORT_Backpacks/F_MED_Swashbuckler_Backpack/Meshes/F_MED_Swashbuckler_Pack_Physics
bodies {
 backpack_root [{shapeType=Capsule, length=4.2559, radius=6.5562, center=<-5.1907 -3.9692 0.0308>, rotation=<rot 0 -87.9634 -89.3475>}]
 dyn_l_sword [{shapeType=Capsule, length=7.4392, radius=3.015, center=<0.1488 0.6344 3.3504>, rotation=<rot -14.7668 -2.6643 10.3476>}]
 dyn_r_sword [{shapeType=Capsule, length=8.2285, radius=3.4805, center=<0.6062 0.3 -2.4996>, rotation=<rot 5.3158 0.6498 -7.7872>}]
}
constraints [
 {
  bone1Name=dyn_l_sword
  bone2Name=backpack_root
  swing1LimitDegrees=0
  swing1Motion=Locked
  swing2LimitDegrees=0
  swing2Motion=Locked
  twistLimitDegrees=45
 }
 {
  bone1Name=dyn_r_sword
  bone2Name=backpack_root
  swing1LimitDegrees=45
  swing1Motion=Locked
  swing2LimitDegrees=45
  swing2Motion=Locked
  twistLimitDegrees=45
 }
]
